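# Ambient water sound cue for the Asteria fountain point of interest, in the outdoor ambient sound class with concurrency capped at three.
# Mixes wave A, looped, with wave B at half volume and no loop flag on its player.
# All quality levels share that mix, which feeds a box-shaped attenuation override.
# A root modulator slightly randomizes pitch at constant volume, over a lowered cue-wide pitch multiplier.
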SoundCue /Game/Sounds/Asteria/POI/Fountain/SC_Fountain_Water_AmbientLoop
root=Modulator_1
nodes [
 {
  id=Attenuation_0
  children [QualityLevel_0]
  properties {bOverrideAttenuation=true, attenuationOverrides {attenuationShape=Box, falloffDistance=700, attenuationShapeExtents=<1600 800 320>}}
 }
 {id=Mixer_0, children [WavePlayer_2,WavePlayer_3], properties {inputVolume [1,0.5]}}
 {id=Modulator_1, children [Attenuation_0], properties {pitchMax=1.1, pitchMin=1, volumeMax=1, volumeMin=1}}
 {id=QualityLevel_0, children [Mixer_0,Mixer_0,Mixer_0,Mixer_0]}
 {id=WavePlayer_2, properties {bLooping=true, soundWaveAssetPtr=&/Fountain_Water_AmbientLoop_A}}
 {id=WavePlayer_3, properties {soundWaveAssetPtr=&/Fountain_Water_AmbientLoop_B}}
]
properties {
 bOverrideConcurrency=true
 maxDistance=2300
 pitchMultiplier=0.8
 soundClassObject=&Fort_Ambient_Outdoors_Athena
 attenuationOverrides {
  attenuationShape=Box
  bAttenuate=false
  bEnableFocusInterpolation=true
  bEnableListenerFocus=true
  falloffDistance=599.8936
  focusAzimuth=25
  nonFocusAzimuth=45
  nonFocusVolumeAttenuation=0.5
  nonSpatializedRadiusStart=300
  stereoSpread=1000
  attenuationShapeExtents=<2263.4798 1387.5758 300>
 }
 concurrencyOverrides {maxCount=3}
 concurrencySet [null]
}
